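# Material function that lets cinematics override an incoming light direction, named for anime-style lighting.
# A ceiled and saturated scalar parameter makes the lerp from the input vector to a cinematic direction parameter act as an on/off toggle.
# A quality switch allows the override only at Default and Epic, passing the input through at Low, Medium and High.
# The UseCinematicLightingExtras static switch bypasses it all when off.
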MaterialFunction /Game/Packages/Fortress_SharedMaterials/Cinematics/MF_CINE_AnimeLightDirOverride
graph {nodes [
  n1 = FunctionInput()
  n2 = VectorParameter() {group=CINEMATICS, parameterName=CIN_LightDirectionVector, sortPriority=1, defaultValue=<rgba 0 0 1 1>}
  n3 = ScalarParameter() {group=CINEMATICS, parameterName=CIN_OverrideLightDirection, sliderMax=1, sortPriority=0}
  n4 = LinearInterpolate(A=@n1, B=@n2.RGB, Alpha=Saturate(None=Ceil(None=@n3)))
  FunctionOutput(None=StaticSwitchParameter(True=QualitySwitch(Default=@n4, Low=@n1, High=@n1, Medium=@n1, Epic=@n4) {desc=has GBuffers}, False=@n1) {group=Cinematics, parameterName=UseCinematicLightingExtras})
 ]}
