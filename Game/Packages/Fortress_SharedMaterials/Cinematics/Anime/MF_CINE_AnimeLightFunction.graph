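# Cinematic anime material function that projects a panning 2D texture onto a virtual plane, apparently for light functions.
# It intersects the normalized world-position direction, traced from the origin, with a plane defined by Location and Rotation (as normal).
# Planar XY coordinates, remapped by UVTiling and offset by ALF_PanningSpeed over time, sample the input texture at an explicit mip level.
# It outputs RGB, alpha and UVs; the UScale and VScale parameters are unconnected.
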MaterialFunction /Game/Packages/Fortress_SharedMaterials/Cinematics/Anime/MF_CINE_AnimeLightFunction
graph {nodes [
  n1 = VectorParameter() {group=CINEMATICS, parameterName=ALF_PanningSpeed, defaultValue=<rgba 0 0 0 1>}
  n2 = Add(A=NamedRerouteUsage(), B=Multiply(A=ComponentMask(None=@n1.RGB) {g=true, r=true}, B=Time()))
  n3 = TextureSample!(
   UVs=@n2
   Tex=FunctionInput(Preview=TextureObject() {texture=&Black}) {inputType=FunctionInput_Texture2D}
   Level
   Apply View MipBias
  ) {mipValueMode=TMVM_MipLevel, samplerSource=SSM_Wrap_WorldGroupSettings, texture=&Black}
  FunctionOutput(None=@n3.RGB) {outputName=RGB}
  FunctionOutput(None=@n3.A) {outputName=ALF_Alpha, sortPriority=1}
  FunctionOutput(None=@n2) {outputName=UVs}
  n4 = FunctionInput(Preview=Constant3Vector() {constant=<rgba 0 0 0 1>}) {inputName=Location}
  n5 = FunctionInput(Preview=Constant3Vector() {constant=<rgba 0 0 -1 1>}) {inputName=Rotation}
  n6 = DotProduct(A=ComponentMask(None=NamedRerouteDeclaration(None=@n4) {name=ALF_Location, nodeColor=<rgba 1 0 0.1406 1>}) {b=true, g=true, r=true}, B=@n5)
  n7 = Normalize(VectorInput=WorldPosition().XYZ)
  n8 = DotProduct(A=Normalize(VectorInput=@n5), B=@n7)
  NamedRerouteDeclaration(None=Multiply(A=Divide(A=@n6, B=@n8), B=@n7)) {name=ALF_PlaneTrace, nodeColor=<rgba 0.6406 0 1 1>}
  n9 = ComponentMask(None=NamedRerouteUsage()) {g=true, r=true}
  n10 = FunctionInput(Preview=Constant2Vector() {g=1, r=1}) {inputName=UVTiling, inputType=FunctionInput_Vector2}
  n11 = Subtract(A=ComponentMask(None=NamedRerouteUsage()) {g=true, r=true}, B=@n10)
  NamedRerouteDeclaration(None=Divide(A=Subtract(A=@n9, B=@n11), B=Multiply(A=@n10) {constB=2})) {name=ALF_UVs, nodeColor=<rgba 0.1016 1 0 1>}
  ScalarParameter() {defaultValue=4, group=CINEMATICS, parameterName=ALF_UScale, sortPriority=3}
  ScalarParameter() {defaultValue=4, group=CINEMATICS, parameterName=ALF_VScale, sortPriority=3}
 ]}
