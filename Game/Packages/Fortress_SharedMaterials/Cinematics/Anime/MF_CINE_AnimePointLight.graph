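# Material function that computes a fake point light mask for anime-style cinematic materials, from location, radius, falloff and multiplier inputs.
# Pixel distance over radius is inverted and smoothstepped, with falloff tightening the edge, and also outputs as Gradient.
# The main output multiplies this by a hard radius cutoff, a saturated dot of pixel normal and unnormalized light vector, and the multiplier, then saturates.
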MaterialFunction /Game/Packages/Fortress_SharedMaterials/Cinematics/Anime/MF_CINE_AnimePointLight
graph {nodes [
  n1 = FunctionInput() {inputName=Falloff, inputType=FunctionInput_Scalar}
  n2 = FunctionInput() {inputName=Location}
  n3 = WorldPosition()
  n4 = Distance(A=@n2, B=@n3.XYZ)
  n5 = FunctionInput() {inputName=Radius, inputType=FunctionInput_Scalar}
  n6 = Saturate(None=Divide(A=@n4, B=@n5))
  n7 = Saturate(None=SmoothStep(Min=Constant(), Max=OneMinus(None=@n1), Value=OneMinus(None=@n6)))
  n8 = Constant()
  n9 = Constant() {r=1}
  n10 = DotProduct(A=PixelNormalWS(), B=Subtract(A=@n2, B=@n3.XYZ))
  n11 = Multiply(A=@n7, B=Multiply(A=If(A=@n4, B=@n5, "A > B"=@n8, "A == B"=@n9, "A < B"=@n9), B=Saturate(None=@n10)))
  n12 = FunctionInput() {inputName=Multiplier, inputType=FunctionInput_Scalar}
  FunctionOutput(None=Saturate(None=Multiply(A=@n11, B=@n12)))
  FunctionOutput(None=@n7) {outputName=Gradient}
 ]}
