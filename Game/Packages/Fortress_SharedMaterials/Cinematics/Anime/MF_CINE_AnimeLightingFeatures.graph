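# Material function that adds fake cinematic point lights for anime-style shading, outputting light sources, light color and a shading-curve adjustment.
# Primary and alternate lights at parameter-driven world positions come from MF_CINE_AnimePointLight and are summed, with comparisons choosing between their intensity, color and shading settings.
# Each light has an optional textured light function with contrast and two-color blending.
# A static switch gates the outputs, and quality switches neutralize the effect at Low, Medium and High.
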MaterialFunction /Game/Packages/Fortress_SharedMaterials/Cinematics/Anime/MF_CINE_AnimeLightingFeatures
graph {nodes [
  n1 = VectorParameter() {desc=Using BP world position, group=CINEMATICS, parameterName=CIN_LightSource, defaultValue=<rgba -2472.0681 18462.6211 2010.4589 1>}
  n2 = MaterialFunctionCall!(
   Location=@n1.RGB
   Falloff=ScalarParameter() {group=CINEMATICS, parameterName=CIN_LightSourceFalloff}
   Multiplier=ScalarParameter() {defaultValue=1, group=CINEMATICS, parameterName=CIN_EnableFakePointLight}
   Radius=ScalarParameter() {defaultValue=30, group=CINEMATICS, parameterName=CIN_LightSourceRadius}
  ) {materialFunction=&/MF_CINE_AnimePointLight}
  n3 = NamedRerouteDeclaration(None=Add(A=NamedRerouteDeclaration(None=@n2.Result) {name=MainLightSource, nodeColor=<rgba 0 1 0.1328 1>}, B=NamedRerouteUsage())) {name=CIN_LightSource, nodeColor=<rgba 1 0.1875 0 1>}
  n4 = Constant()
  FunctionOutput(None=StaticSwitchParameter(True=QualitySwitch(Default=@n3, Low=@n4, High=@n4, Medium=@n4, Epic=@n3) {desc=has GBuffers}, False=Constant()) {group=CINEMATICS, parameterName=EnableCinematicLightingFeatures}) {outputName=CINE_LightSources, sortPriority=1}
  n5 = ScalarParameter() {defaultValue=100, group=CINEMATICS, parameterName=CIN_LightSource_Intensity}
  n6 = Multiply(
   A=If(
    A=NamedRerouteUsage()
    B=NamedRerouteUsage()
    "A > B"=@n5
    "A == B"=@n5
    "A < B"=ScalarParameter() {defaultValue=100, group=CINEMATICS, parameterName=CIN_LightSourceAlt_Intensity}
   )
   B=NamedRerouteUsage()
  )
  n7 = Constant() {r=1}
  FunctionOutput(None=StaticSwitchParameter(True=QualitySwitch(Default=@n6, Low=@n7, High=@n7, Medium=@n7, Epic=@n6), False=@n7) {group=CINEMATICS, parameterName=EnableCinematicLightingFeatures}) {outputName=CINE_LightColor, sortPriority=2}
  n8 = FunctionInput()
  n9 = Constant()
  n10 = NamedRerouteUsage()
  n11 = NamedRerouteUsage()
  n12 = ScalarParameter() {defaultValue=300, group=CINEMATICS, parameterName=CIN_LightSource_ShadingReduction}
  n13 = ScalarParameter() {defaultValue=300, group=CINEMATICS, parameterName=CIN_LightSourceAlt_ShadingReduction}
  n14 = QualitySwitch(Default=Multiply(A=NamedRerouteUsage(), B=If(A=@n10, B=@n11, "A > B"=@n12, "A == B"=@n12, "A < B"=@n13)), Low=@n9, High=@n9, Medium=@n9) {desc=has GBuffers}
  FunctionOutput(None=StaticSwitchParameter(True=Add(A=@n8, B=@n14), False=@n8) {group=CINEMATICS, parameterName=EnableCinematicLightingFeatures}) {outputName=CINE_ShadingCurve}
  n15 = NamedRerouteUsage()
  n16 = NamedRerouteUsage()
  n17 = ComponentMask(None=MaterialFunctionCall(In=NamedRerouteUsage(), Contrast=ScalarParameter() {group=CINEMATICS, parameterName=ALF_Mono_Contrast}) {materialFunction=&CheapContrast}.Result) {r=true}
  n18 = LinearInterpolate(A=NamedRerouteUsage(), B=NamedRerouteUsage(), Alpha=@n17)
  n19 = LinearInterpolate(A=NamedRerouteUsage(), B=NamedRerouteUsage(), Alpha=NamedRerouteUsage())
  n20 = Ceil(None=ScalarParameter() {group=CINEMATICS, parameterName=CIN_UseColoredLightFunctionAlpha})
  n21 = ScalarParameter() {group=CINEMATICS, parameterName=CIN_UseColoredLightFunction}
  n22 = LinearInterpolate(
   A=NamedRerouteUsage()
   B=LinearInterpolate(A=LinearInterpolate(A=@n15, B=@n16, Alpha=@n17), B=LinearInterpolate(A=@n18, B=@n19, Alpha=@n20), Alpha=Ceil(None=@n21))
   Alpha=Ceil(None=ScalarParameter() {group=CINEMATICS, parameterName=CIN_UseAnimeLightFunction})
  )
  n23 = NamedRerouteUsage()
  n24 = NamedRerouteUsage()
  n25 = ComponentMask(None=MaterialFunctionCall(In=@n24, Contrast=ScalarParameter() {group=CINEMATICS, parameterName=ALF2_Mono_Contrast}) {materialFunction=&CheapContrast}.Result) {r=true}
  n26 = NamedRerouteUsage()
  n27 = NamedRerouteUsage()
  n28 = ScalarParameter() {group=CINEMATICS, parameterName=CIN_UseColoredLightFunctionAlpha2}
  n29 = LinearInterpolate(
   A=LinearInterpolate(A=NamedRerouteUsage(), B=NamedRerouteUsage(), Alpha=@n25)
   B=LinearInterpolate(A=LinearInterpolate(A=@n26, B=@n24, Alpha=@n25), B=LinearInterpolate(A=@n26, B=@n24, Alpha=@n27), Alpha=Ceil(None=@n28))
   Alpha=Ceil(None=ScalarParameter() {group=CINEMATICS, parameterName=CIN_UseColoredLightFunction2})
  )
  n30 = Ceil(None=ScalarParameter() {group=CINEMATICS, parameterName=CIN_UseAnimeLightFunction2})
  NamedRerouteDeclaration(None=If(A=NamedRerouteUsage(), B=NamedRerouteUsage(), "A > B"=@n22, "A == B"=@n22, "A < B"=LinearInterpolate(A=@n23, B=@n29, Alpha=@n30))) {name=AnimeLightFunction, nodeColor=<rgba 0.8047 0 1 1>}
  n31 = VectorParameter() {desc=Using BP world position, group=CINEMATICS, parameterName=CIN_LightSourceAlt, defaultValue=<rgba -2472.0681 18462.6211 2010.4589 1>}
  n32 = MaterialFunctionCall(
   Location=@n31.RGB
   Falloff=ScalarParameter() {group=CINEMATICS, parameterName=CIN_LightSourceAlt_Falloff}
   Multiplier=ScalarParameter() {group=CINEMATICS, parameterName=CIN_EnableAltLightSource}
   Radius=ScalarParameter() {defaultValue=800, group=CINEMATICS, parameterName=CIN_LightSourceAlt_Radius}
  ) {materialFunction=&MF_CINE_AnimePointLight}
  NamedRerouteDeclaration(None=@n32.Result) {name=AltLightSource, nodeColor=<rgba 1 0 0.7734 1>}
  NamedRerouteDeclaration(None=VectorParameter() {group=CINEMATICS, parameterName=CIN_LightSourceColor, defaultValue=<rgba 1 1 1 1>}.RGB) {name=Primary Point Light Color, nodeColor=<rgba 1 0.3516 0 1>}
  n33 = VectorParameter() {group=CINEMATICS, parameterName=CIN_LightSource_SecondaryColor, defaultValue=<rgba 0.5156 0.302 0.0833 1>}
  NamedRerouteDeclaration(None=@n33.RGB) {name=ALF_SecondaryColor, nodeColor=<rgba 0 0.2031 1 1>}
  n34 = ScalarParameter() {group=CINEMATICS, parameterName=ALF_Mono_Contrast}
  NamedRerouteDeclaration(None=If(
    A=NamedRerouteUsage()
    B=NamedRerouteUsage()
    "A > B"=@n34
    "A == B"=@n34
    "A < B"=ScalarParameter() {group=CINEMATICS, parameterName=ALF2_Mono_Contrast}
   )) {name=ALF_Contrast, nodeColor=<rgba 0.6641 0 1 1>}
  n35 = NamedRerouteUsage()
  NamedRerouteDeclaration(None=If(A=NamedRerouteUsage(), B=NamedRerouteUsage(), "A > B"=@n35, "A == B"=@n35, "A < B"=NamedRerouteUsage())) {name=ALF_AlphaComposite, nodeColor=<rgba 0.3359 1 0 1>}
  n36 = NamedRerouteUsage()
  NamedRerouteDeclaration(None=If(A=NamedRerouteUsage(), B=NamedRerouteUsage(), "A > B"=@n36, "A == B"=@n36, "A < B"=NamedRerouteUsage())) {name=ALF_RGB, nodeColor=<rgba 0.2422 0 1 1>}
  n37 = VectorParameter() {desc=Using BP world position, group=CINEMATICS, parameterName=CIN_LightSource, defaultValue=<rgba -2472.0681 18462.6211 2010.4589 1>}
  n38 = MaterialFunctionCall!(
   Rotation=VectorParameter() {group=CINEMATICS, parameterName=ALF_Rotation, defaultValue=<rgba 1 0 0 1>}.RGB
   In=TextureObjectParameter() {group=CINEMATICS, parameterName=AnimeLightFunctionMask, texture=&Black}
   UVTiling=AppendVector(
    A=ScalarParameter() {defaultValue=4, group=CINEMATICS, parameterName=ALF_UScale, sortPriority=3}
    B=ScalarParameter() {defaultValue=4, group=CINEMATICS, parameterName=ALF_VScale, sortPriority=3}
   )
   Location=@n37.RGB
  ) {materialFunction=&/MF_CINE_AnimeLightFunction}
  NamedRerouteDeclaration(None=@n38.ALF_Alpha) {name=ALF_Alpha, nodeColor=<rgba 0 1 0.5781 1>}
  NamedRerouteDeclaration(None=Multiply(A=@n38.RGB, B=NamedRerouteUsage())) {name=ALF_Color, nodeColor=<rgba 0 1 0.9531 1>}
  n39 = VectorParameter() {desc=Using BP world position, group=CINEMATICS, parameterName=CIN_LightSourceAlt, defaultValue=<rgba -2472.0681 18462.6211 2010.4589 1>}
  n40 = MaterialFunctionCall(
   Rotation=VectorParameter() {group=CINEMATICS, parameterName=ALF2_Rotation, defaultValue=<rgba 1 0 0 1>}.RGB
   In=TextureObjectParameter() {group=CINEMATICS, parameterName=AnimeLightFunctionMask2, texture=&Black}
   UVTiling=AppendVector(
    A=ScalarParameter() {defaultValue=4, group=CINEMATICS, parameterName=ALF2_UScale, sortPriority=3}
    B=ScalarParameter() {defaultValue=4, group=CINEMATICS, parameterName=ALF2_VScale, sortPriority=3}
   )
   Location=@n39.RGB
  ) {materialFunction=&MF_CINE_AnimeLightFunction}
  NamedRerouteDeclaration(None=@n40.ALF_Alpha) {name=ALF2_Alpha, nodeColor=<rgba 0.3359 1 0 1>}
  NamedRerouteDeclaration(None=@n40.RGB) {name=ALF2_RGB, nodeColor=<rgba 0.125 0 1 1>}
  NamedRerouteDeclaration(None=@n2.Gradient) {name=Primary Light Gradient, nodeColor=<rgba 0 0.9531 1 1>}
  NamedRerouteDeclaration(None=@n32.Gradient) {name=Secondary Point Light Gradient, nodeColor=<rgba 1 0 0.2578 1>}
  NamedRerouteDeclaration(None=VectorParameter() {group=CINEMATICS, parameterName=CIN_LightSourceAlt_Color, defaultValue=<rgba 1 1 1 1>}.RGB) {name=Alt Light Color, nodeColor=<rgba 0 1 0.2969 1>}
  n41 = VectorParameter() {group=CINEMATICS, parameterName=CIN_LightSourceAlt_SecondaryColor, defaultValue=<rgba 0.5156 0.302 0.0833 1>}
  NamedRerouteDeclaration(None=@n41.RGB) {name=ALF_AltSecondaryColor, nodeColor=<rgba 1 0 0.0938 1>}
 ]}
