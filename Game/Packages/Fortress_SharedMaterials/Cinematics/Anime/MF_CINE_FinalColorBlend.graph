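# Material function that applies the final cinematic lighting color blend to an emissive input, filed under the anime cinematics shared materials.
# CINE_LightSources blends a lit intensity parameter toward LightingExtrasColorData; the result multiplies the emissive, mixed in by ShadingCurve.
# The EnableCinematicLightingFeatures static switch gates the effect.
# A quality switch described as "has GBuffers" keeps it on Default and Epic, passing emissive through unchanged on Low, Medium and High.
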MaterialFunction /Game/Packages/Fortress_SharedMaterials/Cinematics/Anime/MF_CINE_FinalColorBlend
graph {nodes [
  n1 = FunctionInput() {inputName=Emissive In}
  n2 = ScalarParameter() {defaultValue=1, group=CINEMATICS, parameterName=CIN_Lit_intensity}
  n3 = FunctionInput() {inputName=LightingExtrasColorData, sortPriority=1}
  n4 = FunctionInput() {inputName=CINE_LightSources, sortPriority=2}
  n5 = LinearInterpolate(A=@n1, B=Multiply(A=@n1, B=LinearInterpolate(A=@n2, B=@n3, Alpha=@n4)), Alpha=FunctionInput() {inputName=ShadingCurve, sortPriority=3})
  FunctionOutput(None=StaticSwitchParameter(True=QualitySwitch(Default=@n5, Low=@n1, High=@n1, Medium=@n1, Epic=@n5) {desc=has GBuffers}, False=@n1) {group=Cinematics, parameterName=EnableCinematicLightingFeatures})
 ]}
