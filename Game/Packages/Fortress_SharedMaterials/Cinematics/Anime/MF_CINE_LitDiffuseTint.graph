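# Material function for cinematic anime materials that adjusts a lit base color with a key-light tint.
# It desaturates the input by a scalar parameter, then multiplies it by the CIN_KeyTint color.
# The EnableCinematicLightingFeatures static switch gates the effect, and a quality switch applies it only at Default and Epic, passing the input through at Low, Medium and High.
# An unconnected shading path switch is left in the graph.
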MaterialFunction /Game/Packages/Fortress_SharedMaterials/Cinematics/Anime/MF_CINE_LitDiffuseTint
graph {nodes [
  n1 = FunctionInput() {inputName=LitBaseColorIn}
  n2 = Multiply(
   A=Desaturation(None=@n1, Fraction=ScalarParameter() {group=CINEMATICS, parameterName=CIN_Lit_Desat})
   B=VectorParameter() {group=CINEMATICS, parameterName=CIN_KeyTint, defaultValue=<rgba 1 1 1 1>}.RGB
  )
  FunctionOutput(None=StaticSwitchParameter(True=QualitySwitch(Default=@n2, Low=@n1, High=@n1, Medium=@n1, Epic=@n2) {desc=has GBuffers}, False=@n1) {group=CINEMATICS, parameterName=EnableCinematicLightingFeatures})
  ShadingPathSwitch()
 ]}
