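# Material function for cinematic anime materials that desaturates and tints the incoming shaded base color.
# CIN_Shaded_Desat drives the desaturation, and the result is multiplied by the CIN_FillTint color.
# The EnableCinematicLightingFeatures static switch gates the effect, passing the input through when off.
# Within it, a quality switch applies the effect at Default and Epic quality and passes the input through at Low, Medium and High.
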MaterialFunction /Game/Packages/Fortress_SharedMaterials/Cinematics/Anime/MF_CINE_ShadedDiffuseTint
graph {nodes [
  n1 = FunctionInput() {inputName=ShadedBaseColorIn}
  n2 = Multiply(
   A=VectorParameter() {group=CINEMATICS, parameterName=CIN_FillTint, defaultValue=<rgba 1 1 1 1>}.RGB
   B=Desaturation(None=@n1, Fraction=ScalarParameter() {group=CINEMATICS, parameterName=CIN_Shaded_Desat, sliderMax=2})
  )
  FunctionOutput(None=StaticSwitchParameter(True=QualitySwitch(Default=@n2, Low=@n1, High=@n1, Medium=@n1, Epic=@n2) {desc=Only available on Epic Quality}, False=@n1) {group=CINEMATICS, parameterName=EnableCinematicLightingFeatures})
 ]}
