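# CINE_MF_EnvPostFX is a placeholder material function that gives cinematics a hook to inject functionality into environment materials.
# It passes its material attributes input straight to its output, unchanged.
# Its description states that it should always be empty in FN.
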MaterialFunction /Game/Packages/Fortress_SharedMaterials/Cinematics/CINE_MF_EnvPostFX
graph {
 properties {description="A placeholder function for cinematics to inject functionality.  Should always be empty in FN"}
 nodes [FunctionOutput(None=FunctionInput() {inputType=FunctionInput_MaterialAttributes})]
}
